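# Editor utility Blueprint for environment art that visualizes the pivots of actors selected in the level.
# The construction script starts a looping timer that calls UpdatePivotVisuals by name, so the event graph holds no logic.
# Each call gets the selection from EditorActorSubsystem and draws a debug coordinate system at every actor's location and rotation.
# The draw duration matches the timer interval, so the axes refresh continuously.
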EditorUtilityBlueprint /Game/ArtTools/Environment/PivotRenderer/EU_PivotRenderer
graphs [
 EventGraph {nodes [
   Event BeginPlay()
   Event ActorBeginOverlap()
   Event Tick!() -> (OutputDelegate, then, DeltaSeconds:float)
  ]}
 UpdatePivotVisuals {nodes [
   UpdatePivotVisuals() {is=title K2Node_FunctionEntry}
   > n1 = GetSelectedLevelActors(self=Get EditorActorSubsystem())
   > n2 = For Each Loop(Array:Array of Actor Object References=@n1) -> (Array Element:Actor Object Reference)
   n3 = BreakTransform(InTransform=Get Actor Transform(self=@n2))
   > DrawDebugCoordinateSystem!(
    execute
    WorldContextObject:Object Reference
    AxisLoc:Vector=@n3
    AxisRot:Rotator=@n3.Rotation
    Scale:float=150
    Duration:float=0.1
    Thickness:float=8
    DepthPriority:EDrawDebugSceneDepthPriorityGroup Enum=World
   ) -> (then)
  ]}
 UserConstructionScript {nodes [
   Construction Script()
   > Set Timer by Function Name!(
    execute
    Object:Object Reference
    FunctionName:String=UpdatePivotVisuals
    Time:float=0.1
    bLooping:bool=true
    bMaxOncePerFrame:bool=false
    InitialStartDelay:float=0
    InitialStartDelayVariance:float=0
   ) -> (then, ReturnValue:Timer Handle Structure) {is=title K2Node_CallFunction}
  ]}
]
variables []
